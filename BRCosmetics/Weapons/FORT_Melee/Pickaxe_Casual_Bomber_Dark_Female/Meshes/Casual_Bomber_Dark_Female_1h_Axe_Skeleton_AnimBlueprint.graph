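# Post-process-style Anim Blueprint for the Casual Bomber Dark Female one-handed axe pickaxe, built to scale its horns and add jaw physics.
# HornScale lerps between MinHornScale and MaxHornScale, feeding additive scale on the main, left and right horn bones.
# A low-alpha AnimDynamics spring then moves dyn_jaw within angular limits.
# As wired, Output Pose takes Input Pose directly, bypassing that chain; the empty EventGraph never sets HornScale.
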AnimBlueprint /BRCosmetics/Weapons/FORT_Melee/Pickaxe_Casual_Bomber_Dark_Female/Meshes/Casual_Bomber_Dark_Female_1h_Axe_Skeleton_AnimBlueprint
target_skeleton=&/Casual_Bomber_Dark_Female_1h_Axe_Skeleton
graphs [
 AnimGraph {nodes [
   n1 = "Lerp (Vector)"!(
    A:Vector=Get MinHornScale() -> (MinHornScale:Vector)
    B:Vector=Get MaxHornScale() -> (MaxHornScale:Vector)
    Alpha:float=Get HornScale() -> (HornScale:double)
   ) -> (ReturnValue:Vector)
   n2 = "Transform (Modify) Bone - Bone: dyn_main_horn"(ComponentPose=Local To Component(), Alpha=1, Translation="0,0,0", Rotation="0,0,0", Scale=@n1) {scaleMode=BMM_Additive, alphaBoolBlend {blendOption=Linear}, boneToModify {boneName=dyn_main_horn}}
   n3 = Anim Dynamics - Bone: dyn_jaw(
    ComponentPose="Transform (Modify) Bone - Bone: dyn_r_horn"(
     ComponentPose="Transform (Modify) Bone - Bone: dyn_l_horn"(ComponentPose=@n2, Alpha=1, Translation="0,0,0", Rotation="0,0,0", Scale=@n1) {scaleMode=BMM_Additive, alphaBoolBlend {blendOption=Linear}, boneToModify {boneName=dyn_l_horn}}
     Alpha=1
     Translation="0,0,0"
     Rotation="0,0,0"
     Scale=@n1
    ) {scaleMode=BMM_Additive, alphaBoolBlend {blendOption=Linear}, boneToModify {boneName=dyn_r_horn}}
    Alpha=0.1
    ExternalForce="0,0,0"
   ) {
    angularDampingOverride=0.99
    angularSpringConstant=500
    bAngularSpring=true
    bGravityOverrideInSimSpace=true
    bOverrideAngularDamping=true
    gravityScale=0.1
    lODThreshold=1
    alphaBoolBlend {blendOption=Linear}
    boundBone {boneName=dyn_jaw}
    componentAppliedLinearAccClamp=<1000 1000 1000>
    componentLinearAccScale=<1 1 1>
    componentLinearVelScale=<1 1 1>
    physicsBodyDefinitions [{
      sphereCollisionRadius=0
      boundBone {boneName=dyn_jaw}
      boxExtents=<8 8 2>
      constraintSetup {twistAxis=AxisY, angularLimitsMax=<35 0 0>, angularLimitsMin=<-15 0 0>, angularTarget=<0 1 0>}
      localJointOffset=<0 -5.25 -5>
     }]
   }
   Component To Local(ComponentPose=@n3)
   Output Pose(Result=Input Pose()) {layerGroup=None}
  ]}
 EventGraph {links [], nodes []}
]
variables [
 HornScale {type=number}
 MaxHornScale {type=Vector, default=<1.5 1.5 1.5>}
 MinHornScale {type=Vector, default=<1 1 1>}
]
